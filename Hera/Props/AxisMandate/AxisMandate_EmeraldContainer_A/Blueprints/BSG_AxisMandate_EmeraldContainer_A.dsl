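; Emerald container prop Blueprint that bobs its gem mesh (NarrowFlea_Gem_DS) as a cosmetic effect tied to player overlap.
; Begin overlap checks for the local player's controlled pawn, handling pawn caching and the StartTracks call.
; Start Tracks runs a looping timer whose tick nudges the gem vertically by a sine of actor lifetime, scaled by Speed.
; End overlap by the cached pawn clears that timer.
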(event Custom|Start%20Tracks
  (Networking|CanExecuteCosmeticEvents
    (:True
      (bind _returnvalue (Utilities|Time|SetTimerbyEvent (delegate Custom|Timer%20Tick) 0.1 true true)))))

(event Custom|Timer%20Tick
  (Transformation|AddWorldOffset (Variables|Default|GetNarrowFlea_Gem_DS) (Math|Vector|MakeVector 0.0 0.0 (* (Math|Trig|Sin(Degrees) (* (Actor|GetGameTimeSinceCreation) (Variables|Default|GetSpeed))) (* (Variables|Default|GetSpeed) 0.01))) false false))

(event EventBeginPlay)

(event Collision|EventActorBeginOverlap
  (params (OtherActor :type Actor%20Object%20Reference))
  ; Overlap
  (Networking|CanExecuteCosmeticEvents
    (:True
      (if (== OtherActor (Pawn|GetControlledPawn (Fortnite|GetFirstLocalFortPlayerController true)))
        (Variables|Default|SetPlayer%20Pawn OtherActor)
        (CallFunction|StartTracks)))))

(event Collision|EventActorEndOverlap
  (params (OtherActor :type Actor%20Object%20Reference))
  ; Overlap
  (Networking|CanExecuteCosmeticEvents
    (:True
      (if (== (Variables|Default|GetPlayer%20Pawn) OtherActor)
        (Utilities|Time|ClearandInvalidateTimerbyHandle (stale Custom|Start%20Tracks _returnvalue))))))

(fn StartTrack)

(fn ConstructionScript)
