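; BP_AirFlow_GenoBase_Fans is an airflow effect actor, apparently for fans in the GenoBase set, built around a cylinder mesh.
; The construction script draws values from a random stream to vary the cylinder's material scalars: time speed, V gradient bias and power, and noise mask exponent and bias.
; The stream allows repeatable variation between instances, and the BeginPlay, overlap and Tick events hold no logic.
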(event EventBeginPlay)

(event Collision|EventActorBeginOverlap
  (params (OtherActor :type Actor%20Object%20Reference)))

(event EventTick
  (params (DeltaSeconds :type Float%20%28single-precision%29)))

(fn ConstructionScript
  (bind _stream (Variables|Default|GetStream))
  (bind _sm_cylinder_airflow_radius (Variables|Default|GetSM_Cylinder_AirFlow_Radius))
  (Rendering|Material|SetScalarParameterValueonMaterials _sm_cylinder_airflow_radius "Time Speed" (Math|Random|RandomFloatinRangefromStream _stream 0.35 0.8))
  (Rendering|Material|SetScalarParameterValueonMaterials _sm_cylinder_airflow_radius "V Gradient Position Bias" (Math|Random|RandomFloatinRangefromStream _stream 1.3 1.7))
  (Rendering|Material|SetScalarParameterValueonMaterials _sm_cylinder_airflow_radius "v Gradient Power" (Math|Random|RandomFloatinRangefromStream _stream 1.0 1.5))
  (Rendering|Material|SetScalarParameterValueonMaterials _sm_cylinder_airflow_radius "Noise Mask Exponent" (Math|Random|RandomFloatinRangefromStream _stream 0.15 0.35))
  (Rendering|Material|SetScalarParameterValueonMaterials _sm_cylinder_airflow_radius "Noise Mask Bias" (Math|Random|RandomFloatinRangefromStream _stream 1.0 0.4)))
